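# Translucent Niagara mesh particle material for the box element of the EnergyNet skydiving contrail, colored and faded by particle color.
# A static switch selects either raw particle color or desaturatable particle color multiplied by the alpha of a curve atlas row sampled along U, feeding a shared emissive adjustment function.
# The same switch selects opacity: particle alpha times either a U-centered parabolic fade toward both ends or a one-sided power ramp.
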Material /BRCosmetics/Athena/Effects/Particles/Skydiving/NS_Contrail_EnergyNet/M_Contrail_EnergyNet_Box
graph {
 properties {bUsedWithNiagaraMeshParticles=true, blendMode=BLEND_Translucent, pixelDepthOffsetMode=PDOM_Legacy, usedShadingModels=Default Lit}
 nodes [
  Constant() {r=1}
  n1 = ParticleColor()
  n2 = CurveAtlasRowParameter(CurveTime=ComponentMask(None=TextureCoordinate()) {r=true}) {atlas=/Game/Athena/LinearColorCurves/VFX_CurveAtlas, curve=/Game/Athena/LinearColorCurves/Curve_Contrail_EnergyNet_Box, defaultValue=22}
  n3 = ParticleColor()
  n4 = Desaturation(None=@n3.RGB, Fraction=ScalarParameter() {parameterName=Particle Color Desaturation})
  MaterialFunctionCall(Emissive=StaticSwitchParameter(True=@n1.RGB, False=Multiply(A=@n2.A, B=@n4)) {parameterName=Ues Particles}) {materialFunction=&MF_FX_EmissiveAdjustments}
  n5 = TextureCoordinate()
  n6 = ComponentMask(None=@n5) {r=true}
  n7 = OneMinus(None=ComponentMask(None=@n5) {r=true})
  n8 = Multiply(A=Subtract(A=Multiply(A=@n6, B=@n7)) {constB=0.05}) {constB=2}
  n9 = ComponentMask(None=TextureCoordinate()) {r=true}
  StaticSwitchParameter(True=Multiply(A=@n1.A, B=Saturate(None=@n8)), False=Multiply(A=@n3.A, B=Power(Base=@n9) {constExponent=1.5})) {parameterName=Ues Particles}
 ]
}
