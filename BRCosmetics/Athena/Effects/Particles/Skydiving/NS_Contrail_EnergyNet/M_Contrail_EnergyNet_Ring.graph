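# Translucent Niagara mesh-particle material for the ring of the EnergyNet skydiving contrail.
# A curve-atlas color sampled along V, shaped by a U-based edge gradient and particle color, passes through an emissive adjustment function.
# An alpha chain multiplies particle alpha, a one-sided U falloff and a powered, panning mask texture on the second UV set.
# Dynamic parameters scale the mask pan speed, while an unused U-gradient product remains in the graph.
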Material /BRCosmetics/Athena/Effects/Particles/Skydiving/NS_Contrail_EnergyNet/M_Contrail_EnergyNet_Ring
graph {
 properties {bUsedWithNiagaraMeshParticles=true, blendMode=BLEND_Translucent, pixelDepthOffsetMode=PDOM_Legacy, usedShadingModels=Default Lit}
 nodes [
  n1 = CurveAtlasRowParameter(CurveTime=ComponentMask(None=TextureCoordinate()) {g=true}) {atlas=/Game/Athena/LinearColorCurves/VFX_CurveAtlas, curve=/Game/Athena/LinearColorCurves/Curve_Contrail_EnergyNet_Box, defaultValue=22}
  n2 = TextureCoordinate()
  n3 = Multiply(A=ComponentMask(None=@n2) {r=true}, B=OneMinus(None=ComponentMask(None=@n2) {r=true}))
  n4 = ScalarParameter() {defaultValue=3, parameterName=Color Hardness}
  n5 = Multiply(A=@n1, B=OneMinus(None=Multiply(A=@n3, B=@n4) {constB=14}))
  n6 = ScalarParameter() {defaultValue=4, parameterName=Emissive Mul}
  n7 = ParticleColor()
  MaterialFunctionCall(Emissive=Multiply(A=Multiply(A=@n5, B=@n6), B=@n7.RGB)) {materialFunction=&MF_FX_EmissiveAdjustments}
  n8 = TextureCoordinate()
  n9 = ComponentMask(None=@n8) {r=true}
  Multiply(A=@n9, B=OneMinus(None=ComponentMask(None=@n8) {r=true}))
  n10 = Saturate(None=Multiply(A=@n9, B=ScalarParameter() {defaultValue=1.5, parameterName=Alpha Hardness}) {constB=14})
  n11 = TextureCoordinate() {coordinateIndex=1}
  n12 = VectorParameter() {group=WPO, parameterName=Mask_UV, channelNames {a=V_Pan, b=U_Pan, g=V_Tilling, r=U_Tilling}, defaultValue=<rgba 0.1 2 1 -0.1>}
  n13 = AppendVector(A=Multiply(A=@n12.U_Tilling, B=ComponentMask(None=@n11) {r=true}), B=Multiply(A=ComponentMask(None=@n11) {g=true}, B=@n12.V_Tilling))
  n14 = DynamicParameter() {paramNames [U_PanSpeed,V_PanSpeed,Param3,Param4]}
  n15 = AppendVector(A=Multiply(A=@n12.U_Pan, B=@n14.U_PanSpeed), B=Multiply(A=@n12.V_Pan, B=@n14.V_PanSpeed))
  n16 = Power(Base=TextureSampleParameter2D(UVs=Panner(Coordinate=@n13, Speed=@n15) {speedX=1, speedY=1}) {group=WPO, parameterName=Mask_Tex, texture=&T_Smooth_tile}.R) {constExponent=1.5}
  Multiply(A=@n7.A, B=Multiply(A=OneMinus(None=@n10), B=Multiply(A=@n16) {constB=2}))
 ]
}
