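# Translucent, two-sided material for the trail of the EnergyNet skydiving contrail, built for Niagara mesh particles.
# A V gradient blended with panned smooth noise samples a curve atlas, squared, plus a tinted strip mask and particle color for emissive.
# Opacity combines panning block and strip masks, a static mask, U-edge and V-end falloffs, particle alpha and depth fade.
# Dynamic parameters supply mask alpha and random offsets to the panning.
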Material /BRCosmetics/Athena/Effects/Particles/Skydiving/NS_Contrail_EnergyNet/M_Contrail_EnergyNet_Trail
graph {
 properties {
  bUsedWithNiagaraMeshParticles=true
  blendMode=BLEND_Translucent
  pixelDepthOffsetMode=PDOM_Legacy
  twoSided=true
  usedShadingModels=Default Lit
 }
 nodes [
  n1 = TextureCoordinate()
  n2 = ComponentMask(None=@n1) {r=true}
  n3 = VectorParameter() {group=WPO, parameterName=Mask_UV, channelNames {a=V_Pan, b=U_Pan, g=V_Tilling, r=U_Tilling}, defaultValue=<rgba 0.75 0.5 0 1.8>}
  n4 = ComponentMask(None=@n1) {g=true}
  n5 = Panner(
   Coordinate=AppendVector(A=Multiply(A=@n3.U_Tilling, B=@n2), B=Multiply(A=@n4, B=@n3.V_Tilling))
   Speed=AppendVector(A=@n3.U_Pan, B=@n3.V_Pan)
  ) {speedX=1, speedY=1}
  n6 = LinearInterpolate(
   A=ComponentMask(None=TextureCoordinate()) {g=true}
   B=TextureSample(UVs=@n5) {texture=&T_Smooth_tile}.R
  ) {constAlpha=0.2}
  n7 = CurveAtlasRowParameter(CurveTime=@n6) {atlas=/Game/Athena/LinearColorCurves/VFX_CurveAtlas, curve=/Game/Athena/LinearColorCurves/Curve_Contrail_EnergyNet_Trail, defaultValue=21}
  n8 = Multiply(A=@n7, B=@n7)
  n9 = TextureCoordinate()
  n10 = ComponentMask(None=@n9) {r=true}
  n11 = VectorParameter() {parameterName=Mask_Strip_UV, channelNames {a=V_Pan, b=U_Offset, g=V_Tile, r=U_Tile}, defaultValue=<rgba 2 0.3 0 1>}
  n12 = ComponentMask(None=@n9) {g=true}
  n13 = DynamicParameter() {defaultValue=<rgba 1 1 1 0>, paramNames [Param1,Param2,MaskAlpha,Random]}
  n14 = Panner(Coordinate=AppendVector(A=Multiply(A=@n10, B=@n11.U_Tile), B=Multiply(A=@n11.V_Tile, B=@n12)), Time=Add(A=@n13.Random, B=@n11.U_Offset)) {speedX=1}
  n15 = Time()
  n16 = Panner(Coordinate=TextureCoordinate() {uTiling=0, vTiling=0}, Time=Multiply(A=@n13.Random, B=Multiply(A=@n11.V_Pan, B=@n15))) {speedY=1}
  n17 = OneMinus(None=TextureSample(UVs=Add(A=@n14, B=@n16)) {texture=&/T_Contrail_EnergyNet_Mask02}.R)
  n18 = Multiply(A=@n17, B=Constant3Vector() {constant=<rgba 0.25 0.075 0.075 0>})
  n19 = ParticleColor()
  MaterialFunctionCall(Emissive=Multiply(A=Add(A=@n8, B=@n18), B=@n19.RGB)) {materialFunction=&MF_FX_EmissiveAdjustments}
  n20 = Power(Base=@n6) {constExponent=1}
  n21 = TextureCoordinate()
  n22 = ComponentMask(None=@n21) {r=true}
  n23 = OneMinus(None=ComponentMask(None=@n21) {r=true})
  n24 = Power(Base=Multiply(A=Multiply(A=@n22, B=@n23)) {constB=2}) {constExponent=6}
  n25 = TextureCoordinate()
  n26 = VectorParameter() {group=WPO, parameterName=Mask_Block_UV, channelNames {a=V_Pan, b=U_Pan, g=V_Tilling, r=U_Tilling}, defaultValue=<rgba 0.7 1.2 0 0.25>}
  n27 = Multiply(A=@n26.U_Tilling, B=ComponentMask(None=@n25) {r=true})
  n28 = Multiply(A=ComponentMask(None=@n25) {g=true}, B=@n26.V_Tilling)
  n29 = Add(A=@n26.V_Pan, B=@n13.Random)
  n30 = TextureSample(UVs=Panner(Coordinate=AppendVector(A=@n27, B=@n28), Speed=AppendVector(A=@n26.U_Pan, B=@n29)) {speedX=1, speedY=1}) {texture=&T_Contrail_EnergyNet_Mask02}
  n31 = TextureCoordinate()
  n32 = Multiply(
   A=LinearInterpolate(A=ComponentMask(None=@n31) {g=true})
   B=TextureSample() {texture=&/T_Contrail_EnergyNet_Mask01}.R
  )
  n33 = Power(Base=TextureSample(UVs=@n5) {texture=&T_Smooth_tile}.R) {constExponent=0.25}
  n34 = LinearInterpolate(
   A=Multiply(A=Multiply(A=@n32, B=@n33), B=DynamicParameter() {defaultValue=<rgba 1 1 1 0>, paramNames [Param1,Param2,MaskAlpha,Random]}.MaskAlpha)
   B=Multiply(A=@n17, B=Constant() {r=3})
  ) {constAlpha=0.4}
  n35 = ComponentMask(None=TextureCoordinate()) {g=true}
  n36 = OneMinus(None=Multiply(A=Subtract(A=@n35) {constB=0.8}) {constB=10})
  n37 = Multiply(A=Multiply(A=Multiply(A=@n20), B=Multiply(A=@n24) {constB=20}), B=Multiply(A=Multiply(A=@n30.R, B=@n34), B=Saturate(None=@n36)))
  n38 = ScalarParameter() {defaultValue=5, parameterName=AlphaMul}
  Multiply(A=@n19.A, B=DepthFade(Opacity=Multiply(A=@n37, B=@n38), FadeDistance=ScalarParameter() {defaultValue=100, parameterName=DepthFade}))
 ]
}
