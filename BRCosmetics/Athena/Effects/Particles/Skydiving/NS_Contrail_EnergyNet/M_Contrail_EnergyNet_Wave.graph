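# Translucent two-sided material for the wave layer of the EnergyNet skydiving contrail, usable on Niagara ribbons and mesh particles.
# Two panning wave-texture layers, warped in V by sine-pulsed smoke noise, combine with a tinted soft smoke layer into blue-to-teal emissive.
# Dynamic parameters drive glow, noise strength and per-particle scroll speed.
# UV edge masks, inverse Fresnel and depth fade soften opacity, while panned noise displaces vertices along normals.
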Material /BRCosmetics/Athena/Effects/Particles/Skydiving/NS_Contrail_EnergyNet/M_Contrail_EnergyNet_Wave
graph {
 properties {
  bUsedWithNiagaraMeshParticles=true
  bUsedWithNiagaraRibbons=true
  blendMode=BLEND_Translucent
  pixelDepthOffsetMode=PDOM_Legacy
  twoSided=true
  usedShadingModels=Default Lit
 }
 nodes [
  n1 = ComponentMask(None=TextureCoordinate()) {g=true}
  n2 = TextureCoordinate()
  n3 = ComponentMask(None=@n2) {r=true}
  n4 = Multiply(A=Multiply(A=ComponentMask(None=@n2) {r=true}, B=OneMinus(None=@n3))) {constB=2}
  n5 = Multiply(A=Power(Base=@n4) {constExponent=4}) {constB=20}
  n6 = Multiply(A=Power(Base=OneMinus(None=@n1)) {constExponent=1.5}, B=@n5)
  n7 = Multiply(
   A=ScalarParameter() {defaultValue=200, parameterName=Emissive_Mul}
   B=DynamicParameter() {paramNames [EmissiveIntensity, NoiseIntensity, Param3, Random]}.EmissiveIntensity
  )
  n8 = Constant3Vector() {constant=<rgba 0 0 1 0>}
  n9 = Constant3Vector() {constant=<rgba 0 2 1.3333 0>}
  n10 = Multiply(A=Power(Base=@n4) {constExponent=1.2}) {constB=1.2}
  n11 = Constant() {r=6}
  n12 = Multiply(
   A=Multiply(
    A=DynamicParameter() {paramNames [EmissiveIntensity, NoiseIntensity, Param3, Random]}.NoiseIntensity
    B=ScalarParameter() {defaultValue=0.8, parameterName=Noise_Intensity}
   )
   B=LinearInterpolate(Alpha=MaterialFunctionCall(Speed=@n11) {materialFunction=&DebugTimeSine}.Result) {constA=-0.5}
  )
  n13 = TextureCoordinate()
  n14 = VectorParameter() {parameterName=Noise_UV, channelNames {a=V_Pan, b=U_Pan, g=V_Tile, r=U_Tile}, defaultValue=<rgba 1.5 0.5 0 1>}
  n15 = AppendVector(A=Multiply(A=ComponentMask(None=@n13) {r=true}, B=@n14.U_Tile), B=Multiply(A=@n14.V_Tile, B=ComponentMask(None=@n13) {g=true}))
  n16 = Time()
  n17 = Multiply(A=@n14.U_Pan, B=@n16)
  n18 = TextureCoordinate() {uTiling=0, vTiling=0}
  n19 = Multiply(A=@n14.V_Pan, B=@n16)
  n20 = TextureSample(UVs=Add(A=Panner(Coordinate=@n15, Time=@n17) {speedX=1}, B=Panner(Coordinate=@n18, Time=@n19) {speedY=1})) {texture=&T_Inky_Smoke_Tile}
  n21 = Multiply(A=@n4, B=Multiply(A=Constant2Vector() {g=1}, B=Multiply(A=@n12, B=@n20.R)))
  n22 = TextureCoordinate()
  n23 = ComponentMask(None=@n22) {r=true}
  n24 = VectorParameter() {parameterName=WaveTex01_UV, channelNames {a=V_Offset, b=U_Pan, g=V_Tile, r=U_Tile}, defaultValue=<rgba 0.3 1.2 -1 -0.1>}
  n25 = ComponentMask(None=@n22) {g=true}
  n26 = Time()
  n27 = DynamicParameter() {paramNames [EmissiveIntensity, NoiseIntensity, Param3, Random]}
  n28 = Panner(
   Coordinate=AppendVector(A=Multiply(A=@n23, B=@n24.U_Tile), B=Multiply(A=@n24.V_Tile, B=@n25))
   Time=Multiply(A=Multiply(A=@n24.U_Pan, B=@n26), B=@n27.Random)
  ) {speedX=1}
  n29 = Panner(Coordinate=TextureCoordinate() {uTiling=0, vTiling=0}, Time=@n24.V_Offset) {speedY=1}
  n30 = TextureSample(UVs=Add(A=@n21, B=Add(A=@n28, B=@n29))) {samplerType=SAMPLERTYPE_LinearColor, texture=&/T_Contrail_EnergyNet_Wave}
  n31 = TextureCoordinate()
  n32 = ComponentMask(None=@n31) {r=true}
  n33 = VectorParameter() {parameterName=WaveTex02_UV, channelNames {a=V_Offset, b=U_Pan, g=V_Tile, r=U_Tile}, defaultValue=<rgba 0.3 1 0.5 0>}
  n34 = ComponentMask(None=@n31) {g=true}
  n35 = Time()
  n36 = Panner(
   Coordinate=AppendVector(A=Multiply(A=@n32, B=@n33.U_Tile), B=Multiply(A=@n33.V_Tile, B=@n34))
   Time=Multiply(A=@n27.Random, B=Multiply(A=@n33.U_Pan, B=@n35))
  ) {speedX=1}
  n37 = Panner(Coordinate=TextureCoordinate() {uTiling=0, vTiling=0}, Time=@n33.V_Offset) {speedY=1}
  n38 = TextureSample(UVs=Add(A=@n21, B=Add(A=@n36, B=@n37))) {samplerType=SAMPLERTYPE_LinearColor, texture=&T_Contrail_EnergyNet_Wave}
  n39 = Add(A=@n30.R, B=@n38.G)
  n40 = Multiply(A=Multiply(A=Multiply(A=@n6, B=@n7), B=LinearInterpolate(A=@n8, B=@n9, Alpha=@n10)), B=@n39)
  n41 = TextureCoordinate()
  n42 = ComponentMask(None=@n41) {r=true}
  n43 = VectorParameter() {parameterName=WaveTex03_UV, channelNames {a=V_Offset, b=U_Pan, g=V_Tile, r=U_Tile}, defaultValue=<rgba 0.5 1 3 -0.15>}
  n44 = ComponentMask(None=@n41) {g=true}
  n45 = Time()
  n46 = Panner(
   Coordinate=AppendVector(A=Multiply(A=@n42, B=@n43.U_Tile), B=Multiply(A=@n43.V_Tile, B=@n44))
   Time=Multiply(A=@n27.Random, B=Multiply(A=@n43.U_Pan, B=@n45))
  ) {speedX=1}
  n47 = Panner(Coordinate=TextureCoordinate() {uTiling=0, vTiling=0}, Time=@n43.V_Offset) {speedY=1}
  n48 = Multiply(
   A=TextureSample(UVs=Add(A=@n46, B=@n47)) {samplerType=SAMPLERTYPE_Grayscale, texture=/Game/Effects/Fort_Effects/Textures/Smoke/T_Smoke_Trail_Soft_NoClamp}.R
   B=Multiply(A=ScalarParameter() {defaultValue=1, parameterName=Wave03_Intensity}) {constB=0.01}
  )
  n49 = Multiply(A=@n48, B=Constant3Vector() {constant=<rgba 0 666.667 333.3336 0>})
  n50 = ParticleColor()
  MaterialFunctionCall(Emissive=Multiply(A=Add(A=@n40, B=@n49), B=@n50.RGB)) {materialFunction=&MF_FX_EmissiveAdjustments}
  n51 = Multiply(A=Multiply(A=@n5, B=Add(A=@n39, B=@n48)) {constA=1}, B=OneMinus(None=Fresnel() {baseReflectFraction=0, exponent=2}))
  n52 = ScalarParameter() {defaultValue=4, parameterName=AlphaMul}
  Multiply(A=@n50.A, B=DepthFade(Opacity=Multiply(A=@n51, B=@n52), FadeDistance=ScalarParameter() {defaultValue=100, parameterName=DepthFade}))
  n53 = TextureCoordinate()
  n54 = VectorParameter() {group=WPO, parameterName=WPO_UV, channelNames {a=V_Pan, b=U_Pan, g=V_Tilling, r=U_Tilling}, defaultValue=<rgba 0.5 0.5 0 1>}
  n55 = Multiply(A=@n54.U_Tilling, B=ComponentMask(None=@n53) {r=true})
  n56 = Multiply(A=ComponentMask(None=@n53) {g=true}, B=@n54.V_Tilling)
  n57 = TextureSampleParameter2D(UVs=Panner(Coordinate=AppendVector(A=@n55, B=@n56), Speed=AppendVector(A=@n54.U_Pan, B=@n54.V_Pan)) {speedX=1, speedY=1}) {group=WPO, parameterName=WPO_Tex, texture=&T_Smooth_tile}
  n58 = VertexNormalWS()
  Reroute(None=Multiply(A=Multiply(A=@n57.RGB, B=@n58), B=ScalarParameter() {defaultValue=10, parameterName=WPO_Intensity}))
 ]
}
